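# Niagara effect for the Candor Dark glider cosmetic, adding wingtip-style ribbon trails and speed lines.
# SpeedLine spawns short-lived sprites on a torus, stretched by speed, with spawn rate driven by a system timer.
# Trail01 builds two bone-anchored anim-trail ribbons from the skeletal mesh, fading with User.ForwardDot and time of day.
# Front-end modifiers retune both emitters for menu display, where trail opacity follows system age instead.
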NiagaraSystem /BRCosmetics/Gadgets/Assets/VinderTech_GliderChute/Glider_Candor_Dark/FX/NS_Glider_Candor_Dark
dependencies {
 usedDataInterfaces [&NiagaraDataInterfaceCurve, &NiagaraDataInterfaceColorCurve]
 usedDynamicInputs []
 usedModules [
  &self:SystemSpawnScript.SystemScriptSource.SystemScriptGraph.NiagaraNodeAssignment_6.SetVariables_#1
  &SystemState
  &self:SystemSpawnScript.SystemScriptSource.SystemScriptGraph.NiagaraNodeAssignment_1.SetVariables_#2
  &self:SystemSpawnScript.SystemScriptSource.SystemScriptGraph.NiagaraNodeAssignment_3.SetVariables_#3
  &self:SystemSpawnScript.SystemScriptSource.SystemScriptGraph.NiagaraNodeAssignment_0.SetVariables_#4
  &self:SystemSpawnScript.SystemScriptSource.SystemScriptGraph.NiagaraNodeAssignment_2.SetVariables_#5
  &FN_Timer
 ]
}
emitters {
 SpeedLine {
  emitterUpdateScript [
   &EmitterState(
    Delay First Loop Only=true
    Inactive Response=::NewEnumerator1
    Life Cycle Mode=::NewEnumerator1
    Loop Delay=1
    Loop Duration=3
    UseLoopDelay=true
   )
   &SpawnRate(
    SpawnRate=&FN_FrontEndFloatModifier(
     Gameplay=&Multiply_Float(
      A=&Multiply_Float(A=200, B=1)
      B=&Lerp_Float(A=-0.8, B=1, Alpha=$System.FN_Timer.Value)
     )
     Front End=60
    )
    Use Spawn Probability=false
   )
  ]
  particleSpawnScript [
   &InitializeParticle(
    Lifetime Max=0.12
    Lifetime Min=0.1
    Lifetime Mode=::NewEnumerator1
    Position Offset=&FN_FrontEndVector3Modifier(Gameplay=<0 0 0>, Front End=<0 170 0>)
    Sprite Size Max=<2 7>
    Sprite Size Min=<2 2>
    Sprite Size Mode=::NewEnumerator2
    UsePositionOffset=true
    Write Material Random=true
   )
   &TorusLocation(
    Align Torus With Axis=true
    Handle Radius=0
    Large Radius:NiagaraFloat=&RandomRangeFloat(
     Minimum=&FN_FrontEndFloatModifier(Gameplay=200, Front End=150)
     Maximum=&FN_FrontEndFloatModifier(Gameplay=250, Front End=150)
    )
    MaskSpawnGroup=false
    Module.Override Randomness=false
    Offset=&FN_FrontEndVector3Modifier(Gameplay=<0 100 0>, Front End=<0 -100 0>)
    Torus Axis=<0 1 0>
   )
   &AddVelocity(Velocity=&RandomRangeVector(Minimum=<0 500 0>, Maximum=<0 500 0>))
   &AddVelocityFromPoint(
    Module.Use Falloff Distance=false
    Origin Offset=&FN_FrontEndVector3Modifier(Gameplay=<0 -200 0>, Front End=<0 -200 0>)
    Velocity Strength=&FN_FrontEndFloatModifier(Gameplay=1500, Front End=300)
   )
   &CameraOffset(Camera Offset Amount=-10)
  ]
  particleUpdateScript [
   &SolveForcesAndVelocity(Clamp Velocity=false, Limit Acceleration=false)
   &ParticleState()
   &ScaleSpriteSizeBySpeed(Max Scale Factor=<1 50>, Min Scale Factor=<0.01 0.05>, Velocity Threshold=1)
   &ScaleColor(
    Scale Alpha=&Multiply_Float(
     A=&FloatFromCurve(FloatCurve=DataInterface() {propertyValues {Curve {keys [
          {interpMode=RCIM_Cubic, tangentMode=RCTM_User, time=0, value=0}
          {interpMode=RCIM_Cubic, tangentMode=RCTM_User, time=0.5, value=1}
          {tangentMode=RCTM_User, time=1, value=0}
         ]}}})
     B=&Multiply_Float(
      A=&RandomRangeFloat(Minimum=0.1, Maximum=0.3)
      B=&FN_FrontEndFloatModifier(Gameplay=1, Front End=0.3)
     )
    )
    ScaleA=true
    ScaleRGB=true
   )
  ]
  rendererClasses [&NiagaraSpriteRendererProperties]
  renderers [{rendererClass=&NiagaraSpriteRendererProperties}]
 }
 Trail01 {
  emitterUpdateScript [
   &EmitterState()
   &FN_SpawnAnimTrail(Number of Anim Trails=2)
  ]
  particleSpawnScript [
   &FN_InitializeAnimTrailBoneIndecies()
   &FN_InitializeAnimTrailBoneIndecies(AnimTrail Selection=1, First Bone or Socket Index=2, Second Bone or Socket Index=3) {module=FN_InitializeAnimTrailBoneIndecies001}
   &FN_InitializeAnimTrail(
    AnimTrailWidth=&FN_FrontEndFloatModifier(Gameplay=1.03, Front End=0.7)
    Color=<rgba 1 1 1 0.2>
    Lifetime=3
   )
   &AddVelocity(Velocity=<0 400 0>)
   /SetVariables() {isSetParameters=true}
   &CameraOffset(Camera Offset Amount=&FN_FrontEndFloatModifier(Gameplay=-20, Front End=0))
  ]
  particleUpdateScript [
   &ParticleState()
   /SetVariables(Particles.Alpha:NiagaraFloat=&ClampFloat(Float=&Add_Float(
      A=$Particles.Alpha
      B=&MakeCustomFloatFromBool(Bool=&SetBoolByFloatComparison(A=$User.ForwardDot, B=0.2), True Float=0.2, False Float=-0.2)
     ))) {isSetParameters=true}
   &ScaleColor(
    Scale Alpha=&FloatFromCurve(
     FloatCurve=DataInterface() {propertyValues {Curve {keys [<at 0 0>,<at 0.1 0.9>,<at 1 0>]}}}
     Scale Curve=&Multiply_Float(
      A=&Multiply_Float(A=1.5, B=&FNTimeOfDayScalar(Night Scale=0.5, Morning Scale=0.3, Evening Scale=0.3))
      B=&FN_FrontEndFloatModifier(
       Gameplay=$Particles.Alpha
       Front End=&FloatFromCurve(
        FloatCurve=DataInterface() {propertyValues {Curve {keys [
            {arriveTangent=4.1785, time=0, value=0}
            <at 1.5 0>
            {arriveTangent=-0.0405, leaveTangent=-0.0405, time=2, value=1}
            {arriveTangent=0.9995, leaveTangent=0.9995, time=3.5, value=1}
           ]}}}
        CurveIndex=$Engine.System.Age
       )
      )
     )
    )
    Scale RGB=&MakeVectorFromLinearColorRGB(Color=<rgba 1 1 1 1>)
    ScaleA=true
    ScaleRGB=true
   )
   &FN_UpdateAnimTrail(SkeletalMesh=$User.SkeletalMesh)
   &SolveForcesAndVelocity(Clamp Velocity=false, Limit Acceleration=false)
   &DynamicMaterialParameters(
    Index 0 Param 1=1
    Index 0 Param 2=1
    Module.Param0WriteEnabled=true
    Module.Param1WriteEnabled=true
    Module.Param2WriteEnabled=true
    Module.Param3WriteEnabled=true
   )
  ]
  rendererClasses [&NiagaraRibbonRendererProperties]
  renderers [{rendererClass=&NiagaraRibbonRendererProperties}]
 }
}
summary {
 systemName=NS_Glider_Candor_Dark
 emitters [
  {emitterName=Trail01, rendererClasses [&NiagaraRibbonRendererProperties]}
  {emitterName=SpeedLine, rendererClasses [&NiagaraSpriteRendererProperties]}
 ]
 userVariables [
  User.Base Color {type=LinearColor, description="", defaultValue {struct=LinearColor, value=<rgba 0.3 0.3 0.3 1.2>}}
  User.bIsFrontEnd {type=NiagaraBool}
  User.bIsFrontEndPreview {type=NiagaraBool}
  User.Delay {type=NiagaraFloat, defaultValue=0, description=""}
  User.ForwardDot {type=NiagaraFloat, defaultValue=0, description=""}
  User.Glitch Base Color {type=LinearColor, description="", defaultValue {struct=LinearColor, value=<rgba 0.2281 0.5 0.4086 1>}}
  User.Glow Color {type=LinearColor, description="", defaultValue {struct=LinearColor, value=<rgba 0 0.3949 1 0.3>}}
  User.IsFrom3D {type=NiagaraFloat, defaultValue=0, description=""}
  User.RibbonIDOffset {type=NiagaraInt32}
  User.SkeletalMesh {
   type=NiagaraDataInterfaceSkeletalMesh
   defaultValue {
    struct=NiagaraExt_VariableValue_DataInterface
    value {dataInterface=&self:NiagaraDataInterfaceSkeletalMesh_0}
   }
  }
 ]
}
